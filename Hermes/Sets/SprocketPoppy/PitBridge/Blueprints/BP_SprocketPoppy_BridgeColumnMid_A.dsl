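; BP_SprocketPoppy_BridgeColumnMid_A is a Blueprint, apparently a middle column piece for the PitBridge in the SprocketPoppy set.
; Its construction script calls ReconstructSocketFiller on its ISM_SocketFiller component, rebuilding that socket filler whenever construction runs.
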(fn ConstructionScript
  (Class|ISMSocketFiller|ReconstructSocketFiller (Variables|Default|GetISM_SocketFiller)))
